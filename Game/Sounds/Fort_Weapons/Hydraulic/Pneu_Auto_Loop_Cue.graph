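# Sound cue for the automatic fire loop of a pneumatic ranged weapon, using the third-person fire sound class and a long-range rifle attenuation.
# A mixer layers a one-shot start sound at half volume with a looping slow pneumatic fire wave at full volume.
# A short fixed delay holds back the loop so the start sound leads into it.
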SoundCue /Game/Sounds/Fort_Weapons/Hydraulic/Pneu_Auto_Loop_Cue
root=Mixer_0
nodes [
 {id=Delay_0, children [WavePlayer_0], properties {delayMax=0.05, delayMin=0.05}}
 {id=Mixer_0, children [Delay_0,WavePlayer_5], properties {inputVolume [1,0.5]}}
 {id=WavePlayer_0, properties {bLooping=true, soundWaveAssetPtr=&/Pneu_Auto_Loop_01_slow}}
 {id=WavePlayer_5, properties {soundWaveAssetPtr=&/Pneu_Auto_Start_01}}
]
properties {
 attenuationSettings=/Game/Sounds/Attenuation/Fort_Weapon_Rifle_Lng_Att
 maxDistance=7500
 soundClassObject=&Fort_WeaponRanged_Fire_3P_PSM
 volumeMultiplier=1.4
}
