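# Sound cue for the end of a pneumatic automatic weapon's firing loop, in the hydraulic weapon sound set.
# A single wave player feeds the cue output directly, with no randomization or modulation nodes.
# The cue lowers volume and pitch, uses long-range rifle attenuation, and routes to a third-person ranged weapon fire sound class.
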SoundCue /Game/Sounds/Fort_Weapons/Hydraulic/Pneu_Auto_Loop_End_Cue
root=WavePlayer_4
nodes [{id=WavePlayer_4, properties {soundWaveAssetPtr=&/Pneu_Auto_Loop_End_01a}}]
properties {
 attenuationSettings=/Game/Sounds/Attenuation/Fort_Weapon_Rifle_Lng_Att
 maxDistance=7500
 pitchMultiplier=0.7
 soundClassObject=&Fort_WeaponRanged_Fire_3P_PSM
 volumeMultiplier=0.6
}
